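# Physics asset for a playground swing set mesh, giving its two hanging swing chains physics bodies and joints.
# A sphere on Root anchors two branches (Chain_1 to Chain_2, Chain_4 to Chain_5), each an upper capsule above a larger sphere.
# Upper joints to Root have a small single-axis swing limit, lower joints zero swing limits, and all share a wider twist limit.
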PhysicsAsset /Game/Packages/Fortress_ExteriorDecos03/Mesh/PlayGround_SwingSet_01_B_S_Physics
bodies {
 Chain_1 [{shapeType=Capsule, length=13.0831, radius=39.8363, center=<-0.2471 -0.1038 -23.3826>, rotation=<rot 0 90 180>}]
 Chain_2 [{shapeType=Sphere, radius=52.2031, center=<-0.1273 -0.0362 -78.1052>}]
 Chain_4 [{shapeType=Capsule, length=16.1628, radius=37.8742, center=<1.7595 -0.1039 -21.7777>, rotation=<rot 0 90 180>}]
 Chain_5 [{shapeType=Sphere, radius=55.029, center=<-0.3038 -0.0362 -75.5324>}]
 Root [{shapeType=Sphere, radius=19.0626, center=<2.6649 -0.1037 255.2069>}]
}
constraints [
 {bone1Name=Chain_1, bone2Name=Root, swing1LimitDegrees=5, swing2LimitDegrees=0, twistLimitDegrees=45}
 {bone1Name=Chain_2, bone2Name=Chain_1, swing1LimitDegrees=0, swing2LimitDegrees=0, twistLimitDegrees=45}
 {bone1Name=Chain_4, bone2Name=Root, swing1LimitDegrees=5, swing2LimitDegrees=0, twistLimitDegrees=45}
 {bone1Name=Chain_5, bone2Name=Chain_4, swing1LimitDegrees=0, swing2LimitDegrees=0, twistLimitDegrees=45}
]
