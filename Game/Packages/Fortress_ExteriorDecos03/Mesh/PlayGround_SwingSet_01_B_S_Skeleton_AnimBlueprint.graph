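# Animation Blueprint for the PlayGround_SwingSet_01_B_S skeleton that adds physics-driven motion to a playground swing set.
# The live graph runs the Input Pose through a RigidBody node simulating in world space to the Output Pose.
# A series of four AnimDynamics nodes on Chain bones and a DefaultSlot node are left unconnected to the output.
# The Event Graph holds no logic.
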AnimBlueprint /Game/Packages/Fortress_ExteriorDecos03/Mesh/PlayGround_SwingSet_01_B_S_Skeleton_AnimBlueprint
target_skeleton=&/PlayGround_SwingSet_01_B_S_Skeleton
graphs [
 AnimGraph {nodes [
   n1 = Anim Dynamics - Bone: Chain_1(ComponentPose=Local To Component(), Alpha=1, ExternalForce="0,0,0") {
    angularSpringConstant=500
    bGravityOverrideInSimSpace=true
    alphaBoolBlend {blendOption=Linear}
    boundBone {boneName=Chain_1}
    chainEnd {boneName=Chain_3}
    physicsBodyDefinitions [{
      sphereCollisionRadius=0
      boundBone {boneName=Chain_1}
      boxExtents=<100 100 40>
      constraintSetup {twistAxis=AxisZ, angularLimitsMax=<45 45 45>, angularLimitsMin=<-45 -45 -45>, angularTarget=<0 0 0>}
      localJointOffset=<0 0 -100>
     }]
   }
   n2 = Anim Dynamics - Bone: Chain_2(ComponentPose=@n1, Alpha=1, ExternalForce="0,0,0") {
    angularSpringConstant=500
    bGravityOverrideInSimSpace=true
    alphaBoolBlend {blendOption=Linear}
    boundBone {boneName=Chain_2}
    chainEnd {boneName=Chain_3}
    physicsBodyDefinitions [{
      sphereCollisionRadius=0
      boundBone {boneName=Chain_2}
      boxExtents=<100 100 40>
      constraintSetup {twistAxis=AxisZ, angularLimitsMax=<45 45 45>, angularLimitsMin=<-45 -45 -45>, angularTarget=<0 0 0>}
      localJointOffset=<0 0 -100>
     }]
   }
   n3 = Anim Dynamics - Bone: Chain_4(ComponentPose=@n2, Alpha=1, ExternalForce="0,0,0") {
    angularSpringConstant=500
    bGravityOverrideInSimSpace=true
    alphaBoolBlend {blendOption=Linear}
    boundBone {boneName=Chain_4}
    chainEnd {boneName=Chain_3}
    physicsBodyDefinitions [{
      sphereCollisionRadius=0
      boundBone {boneName=Chain_4}
      boxExtents=<100 100 40>
      constraintSetup {twistAxis=AxisZ, angularLimitsMax=<45 45 45>, angularLimitsMin=<-45 -45 -45>, angularTarget=<0 0 0>}
      localJointOffset=<0 0 -100>
     }]
   }
   n4 = Anim Dynamics - Bone: Chain_5(ComponentPose=@n3, Alpha=1, ExternalForce="0,0,0") {
    angularSpringConstant=500
    bGravityOverrideInSimSpace=true
    alphaBoolBlend {blendOption=Linear}
    boundBone {boneName=Chain_5}
    chainEnd {boneName=Chain_3}
    physicsBodyDefinitions [{
      sphereCollisionRadius=0
      boundBone {boneName=Chain_5}
      boxExtents=<100 100 40>
      constraintSetup {twistAxis=AxisZ, angularLimitsMax=<45 45 45>, angularLimitsMin=<-45 -45 -45>, angularTarget=<0 0 0>}
      localJointOffset=<0 0 -100>
     }]
   }
   Component To Local(ComponentPose=@n4)
   AnimGraphNode_SubInput_0 = Input Pose()
   n5 = Local To Component(LocalPose=@AnimGraphNode_SubInput_0)
   Output Pose(Result=Component To Local(ComponentPose=RigidBody(ComponentPose=@n5) {simulationSpace=WorldSpace, alphaBoolBlend {blendOption=Linear}})) {layerGroup=None}
   Slot 'DefaultSlot'(Source=Local Space Ref Pose())
  ]}
 EventGraph {nodes [
   TryGetPawnOwner()
   Event BlueprintUpdateAnimation()
  ]}
]
variables []
