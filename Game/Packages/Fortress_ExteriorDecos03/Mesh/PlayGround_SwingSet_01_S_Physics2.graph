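# Physics asset for the playground swing set mesh in the Fortress exterior decoration package, giving its hanging chains bodies and joints for physical swinging.
# A sphere on Root anchors four two-body chains, each an upper capsule constrained to Root and a lower sphere constrained to that capsule.
# All joints share the same angle limit values, but the lower joints set second swing and twist free, so those axes move unrestricted.
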PhysicsAsset /Game/Packages/Fortress_ExteriorDecos03/Mesh/PlayGround_SwingSet_01_S_Physics2
bodies {
 Chain_1 [{shapeType=Capsule, length=2.1161, radius=46.0889, center=<-2.823 -0.1037 -15.4445>, rotation=<rot 0 90 180>}]
 Chain_10 [{shapeType=Capsule, length=10.8331, radius=45.5863, center=<1.7203 1.8935 -7.1942>, rotation=<rot 0 90 180>}]
 Chain_11 [{shapeType=Sphere, radius=53.1629, center=<2.2733 -0.0378 -109.2002>}]
 Chain_2 [{shapeType=Sphere, radius=50.5677, center=<-3.3171 -0.0391 -108.5203>}]
 Chain_4 [{shapeType=Capsule, length=3.5829, radius=45.9915, center=<0.6173 -0.1036 -14.9506>, rotation=<rot 0 90 180>}]
 Chain_5 [{shapeType=Sphere, radius=50.8668, center=<0.836 -0.0383 -108.8437>}]
 Chain_7 [{shapeType=Capsule, length=9.1086, radius=45.4493, center=<0.6958 1.8935 -11.0394>, rotation=<rot 0 90 180>}]
 Chain_8 [{shapeType=Sphere, radius=49.3813, center=<2.6168 -0.4444 -100.2383>}]
 Root [{shapeType=Sphere, radius=33.1556, center=<-4.183 -0.1037 255.7603>}]
}
constraints [
 {bone1Name=Chain_1, bone2Name=Root, swing1LimitDegrees=15, swing2LimitDegrees=45, twistLimitDegrees=45}
 {bone1Name=Chain_4, bone2Name=Root, swing1LimitDegrees=15, swing2LimitDegrees=45, twistLimitDegrees=45}
 {bone1Name=Chain_7, bone2Name=Root, swing1LimitDegrees=15, swing2LimitDegrees=45, twistLimitDegrees=45}
 {bone1Name=Chain_10, bone2Name=Root, swing1LimitDegrees=15, swing2LimitDegrees=45, twistLimitDegrees=45}
 {
  bone1Name=Chain_2
  bone2Name=Chain_1
  swing1LimitDegrees=15
  swing2LimitDegrees=45
  swing2Motion=Free
  twistLimitDegrees=45
  twistMotion=Free
 }
 {
  bone1Name=Chain_5
  bone2Name=Chain_4
  swing1LimitDegrees=15
  swing2LimitDegrees=45
  swing2Motion=Free
  twistLimitDegrees=45
  twistMotion=Free
 }
 {
  bone1Name=Chain_8
  bone2Name=Chain_7
  swing1LimitDegrees=15
  swing2LimitDegrees=45
  swing2Motion=Free
  twistLimitDegrees=45
  twistMotion=Free
 }
 {
  bone1Name=Chain_11
  bone2Name=Chain_10
  swing1LimitDegrees=15
  swing2LimitDegrees=45
  swing2Motion=Free
  twistLimitDegrees=45
  twistMotion=Free
 }
]
