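# Physics asset for a playground swing set skeletal mesh, defining simple collision bodies and joints for two jointed chains.
# A sphere on Root anchors two mirrored two-link chains, Chain_1 to Chain_2 and Chain_4 to Chain_5, each link using a capsule body.
# All four constraints share the same angular limit values, with a narrower first swing limit than the second swing and twist.
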PhysicsAsset /Game/Packages/Fortress_ExteriorDecos03/Mesh/PlayGround_SwingSet_02_S_Physics
bodies {
 Chain_1 [{shapeType=Capsule, length=36.25, radius=33.25, center=<0.4775 -3.7396 -27.662>, rotation=<rot 0 90 180>}]
 Chain_2 [{shapeType=Capsule, length=36.25, radius=46.5, center=<0.4775 -3.7386 -51.8264>, rotation=<rot 0 90 180>}]
 Chain_4 [{shapeType=Capsule, length=36.25, radius=33.25, center=<0.4775 -3.7396 -27.662>, rotation=<rot 0 90 180>}]
 Chain_5 [{shapeType=Capsule, length=36.25, radius=46.5, center=<0.4775 -3.7386 -51.8264>, rotation=<rot 0 90 180>}]
 Root [{shapeType=Sphere, radius=16.1779, center=<0.5139 -3.7371 266.8151>}]
}
constraints [
 {bone1Name=Chain_1, bone2Name=Root, swing1LimitDegrees=15, swing2LimitDegrees=45, twistLimitDegrees=45}
 {bone1Name=Chain_4, bone2Name=Root, swing1LimitDegrees=15, swing2LimitDegrees=45, twistLimitDegrees=45}
 {bone1Name=Chain_2, bone2Name=Chain_1, swing1LimitDegrees=15, swing2LimitDegrees=45, twistLimitDegrees=45}
 {bone1Name=Chain_5, bone2Name=Chain_4, swing1LimitDegrees=15, swing2LimitDegrees=45, twistLimitDegrees=45}
]
